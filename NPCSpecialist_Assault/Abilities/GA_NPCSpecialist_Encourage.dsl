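; Gameplay ability for the Assault NPC specialist that cheers on its converter (the player it follows) and nearby squadmates during combat.
; AI alert reaching Threatened, or converter combat handling tied to damage and weapon-fire event tasks, triggers a cooldown and distance check.
; On success the NPC plays a montage, gameplay cue and speech bubble, then applies the GrantSlap effect to the converter and in-range squadmates, broadcasting a buffed verb for squadmates.
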(event Ability|EventActivateAbility
  (bind _convertcomponent (Variables|InternalData|GetConvertComponent))
  (bind _returnvalue (Convert|GetConverterPawn _convertcomponent))
  (bind _outfortplayerpawn (CallFunction|GetAvatarasValidFortPlayerPawn)
    (:ValidFortPlayerPawn
      (bind _output_get (Variables|InternalData|SetAIPawn _outfortplayerpawn))
      (Utilities|IsValid (Actor|GetComponentbyClass _output_get "/Script/FortniteGame.FortPawnComponent_Convert")
        (:Is%20Valid
          (Variables|InternalData|SetConvertComponent (Actor|GetComponentbyClass _output_get "/Script/FortniteGame.FortPawnComponent_Convert"))
          (bind _asfort_athena_aibot_controller (Utilities|Casting|CastToFortAthenaAIBotController (Pawn|GetController (Variables|InternalData|GetAIPawn)))
            (:then
              (bind _output_get_2 (Variables|InternalData|SetAIController _asfort_athena_aibot_controller))
              (AlertLevel|BindEventtoOnAlertLevelChangedEvent _output_get_2 (EventDispatchers|CreateEvent :!event Custom|AIAlertLevelChanged))
              (Variables|InternalData|SetLastEncourageTime -6e+01)
              (bind _asynctaskproxy (Ability|Tasks|WaitGameplayEvent "(TagName=\"Event.Damage.Received\")" _returnvalue)
                (:then
                  (Variables|InternalData|SetTASK_DamageReceived _asynctaskproxy)
                  (bind _asynctaskproxy_2 (Ability|Tasks|WaitGameplayEvent "(TagName=\"Event.Weapons.Fired\")" _returnvalue)
                    (:then
                      (Variables|InternalData|SetTASK_WeaponsFired _asynctaskproxy_2)
                      (bind _asynctaskproxy_3 (Ability|Tasks|WaitGameplayEvent "(TagName=\"Event.Weapons.Fire\")" _returnvalue)
                        (:then
                          (Variables|InternalData|SetTASK_WeaponsFire _asynctaskproxy_3)
                          (bind _asynctaskproxy_4 (Ability|Tasks|WaitGameplayEvent "(TagName=\"Event.Weapon.Fired\")" _returnvalue)
                            (:then
                              (Variables|InternalData|SetTASK_WeaponFired _asynctaskproxy_4))
                            (:EventReceived
                              _)))
                        (:EventReceived
                          _))
                      (CallFunction|HandleConverterinCombat))
                    (:EventReceived
                      _)))
                (:EventReceived
                  _)))
            (:CastFailed
              _)))
        (:Is%20Not%20Valid
          _))
      (Ability|EndAbility))
    (:AvatarCastFailed
      _)))

(event Ability|EventOnEndAbility
  (params (bWasCancelled :type Boolean))
  (GameplayTasks|EndTask (Variables|InternalData|GetTASK_DamageReceived)))

(event Custom|AIAlertLevelChanged
  (params (BotController :type Fort%20Athena%20AIBot%20Controller%20Object%20Reference) (OldAlertLevel :type EAlertLevel%20Enum :default "Unaware") (NewAlertLevel :type EAlertLevel%20Enum :default "Unaware"))
  (if (Utilities|Enum|Equal(Enum) NewAlertLevel "Threatened")
    (CallFunction|ValidateCanTriggerEncourage)))

(event Custom|HandleConverterInCombat
  (CallFunction|ValidateCanTriggerEncourage))

(event Custom|ValidateCanTriggerEncourage
  (bind (_montagetoplay _animplayrate) (Utilities|Struct|BreakFortGameplayAbilityMontageInfo (Variables|Animation|GetMontageInfo)))
  (bind _convertcomponent (Variables|InternalData|GetConvertComponent))
  (bind _returnvalue (Convert|GetConverterPawn _convertcomponent))
  (if (and (>= (- (Utilities|Time|GetGameTimeinSeconds) (Variables|InternalData|GetLastEncourageTime)) (ScalableFloat|GetValueatLevel (Variables|Default|GetEncourageCooldown))) (<= (Math|Vector|Distance(Vector) (Transformation|GetActorLocation (Variables|InternalData|GetAIPawn)) (Transformation|GetActorLocation (Convert|GetConverterPawn (Variables|InternalData|GetConvertComponent)))) (ScalableFloat|GetValueatLevel (Variables|Default|GetEncourageMaxDistance))))
    (Variables|InternalData|SetLastEncourageTime (Utilities|Time|GetGameTimeinSeconds))
    (Animation|Utilities|PlayAnimMontage (Variables|InternalData|GetAIPawn) _montagetoplay _animplayrate)
    (Ability|ExecuteGameplayCueOnOwner (Variables|Default|GetGC_Encourage))
    (Class|BPFLAICommonFunctions|PlaySpeechBubble (Variables|InternalData|GetAIPawn) (Variables|Default|GetTAG_SpeechBubbleCheer) true true)
    (GameplayEffects|ApplyGameplayEffectToTarget :self (Ability|GetAbilitySystemComponent _returnvalue) :GameplayEffectClass (Variables|Default|GetGE_GrantSlap) :Target (Ability|GetAbilitySystemComponent _returnvalue) :Level 1.0)
    (bind _squadmembers (Class|FortGameModeAthena|GetSquadMembers (Utilities|Casting|CastToFortGameModeAthena (Game|GetGameMode)) (Class|FortPlayerStateAthena|GetSquadId (Utilities|Casting|CastToFortPlayerStateAthena (Variables|/Script/Engine.Controller|GetPlayerState (Pawn|GetController _returnvalue))))))
    (for _array_element _squadmembers
      (if (and (<= (Math|Vector|Distance(Vector) (Transformation|GetActorLocation (PlayerState|GetPawn _array_element)) (Transformation|GetActorLocation (Variables|InternalData|GetAIPawn))) (ScalableFloat|GetValueatLevel (Variables|Default|GetEncourageMaxDistance))) (!= (PlayerState|GetPawn _array_element) (Convert|GetConverterPawn (Variables|InternalData|GetConvertComponent))))
        (GameplayEffects|ApplyGameplayEffectToTarget :self (Ability|GetAbilitySystemComponent (PlayerState|GetPawn _array_element)) :GameplayEffectClass (Variables|Default|GetGE_GrantSlap) :Target (Ability|GetAbilitySystemComponent (PlayerState|GetPawn _array_element)) :Level 1.0)
        (CustomVerb|BroadcastCustomVerbMessage (CustomVerb|MakeCustomVerbMessage (PlayerState|GetPawn _array_element) "(TagName=\"CustomVerb.TeaCake.Specialist.Assault.Buffed\")"))))))
